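# MetaSound patch for NPC music that holds a queued trigger until a beat lands on the chosen quantize division.
# A resettable counter counts OnBeat, and a modulo check against the division, with a fallback of 1, feeds a gate opened by QueueTrigger.
# Cancel or the fired output closes the gate, using a deferred trigger variable for feedback.
# Debug logs the division under the Identifier label when it fires.
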MetaSoundPatch /FortNPCMusic/Audio/Patches/MSP_NPCMusic_Quantizer
inputs [
 Cancel {type=Trigger}
 Debug {type=Bool}
 Identifier {type=String}
 OnBeat {type=Trigger}
 QuantizeDivision {type=Int32}
 QueueTrigger {type=Trigger}
 Reset {type=Trigger}
 ResetCount {type=Int32}
]
outputs [OnTriggerQuantized {type=Trigger}]
pages [Default {
  input_defaults {Cancel=null, Debug=null, Identifier="", OnBeat=null, QuantizeDivision=4, QueueTrigger=null, Reset=null, ResetCount=64}
  nodes [
   n1 = VariableDeferredAccessor.Trigger(Variable=InitVariable.Trigger().Variable)
   TriggerAnyTriggerAny2_1 = "TriggerAny.Trigger Any (2)"(In 1=^Cancel, In 0=@n1.Value)
   TriggerCounter_0 = UE.Trigger Counter(In=^OnBeat, Reset=^Reset, Reset Count=^ResetCount)
   n2 = TriggerCompare.Clamp:Int32(A=^QuantizeDivision, Compare=@TriggerCounter_0.On Trigger)
   n3 = Value.Value:Int32(Init Value=^QuantizeDivision, Reset=@n2.False, Set=@n2.True) {also {defaults {Target Value=1}}}
   TriggerAnyTriggerAny2_0 = "TriggerAny.Trigger Any (2)"(In 1=@n3.On Reset, In 0=@n3.On Set)
   TriggerControl_0 = UE.Trigger Control(
    Open=^QueueTrigger
    Close=@TriggerAnyTriggerAny2_1.Out
    Trigger In=TriggerCompare.Clamp:Int32(A=UE.Modulo:Int32(PrimaryOperand=@TriggerCounter_0.Count, AdditionalOperands=@n3.Output Value).Out, Compare=@TriggerAnyTriggerAny2_0.Out).True
   )
   n4 = VariableMutator.Trigger(Value=@TriggerControl_0.Trigger Out, Variable=@n1.Variable)
   PrintLogPrintLog_0 = Print Log.Print Log:Int32(
    Label=^Identifier
    Value To Log=^QuantizeDivision
    Trigger=TriggerCompare.Clamp:Bool(A=^Debug, Compare=VariableAccessor.Trigger(Variable=@n4.Variable).Value) {also {defaults {B="<MetaSoundBuilderResult.SUCCEEDED: 0>"}}}.True
   )
  ]
  links [{dst="@OnTriggerQuantized", src=TriggerControl_0.Trigger Out}]
 }]
