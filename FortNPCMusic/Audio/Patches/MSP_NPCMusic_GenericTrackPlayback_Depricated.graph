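# MetaSound patch, named as deprecated, that plays an NPC music track as looping Drums, Else and Ext stereo stems with an optional transition-in stinger.
# QueueTrigger starts either the stems or, if valid, the stinger, whose playback queues the stems; Quantizer patches align starts and stops to OnBeat.
# Stems get low-pass filters and gains, fading in only without a stinger.
# ResetAndStop triggers a quantized fade-out, then stops the loops.
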MetaSoundPatch /FortNPCMusic/Audio/Patches/MSP_NPCMusic_GenericTrackPlayback_Depricated
inputs [
 Crossfade.FadeInTime {type=Time}
 Crossfade.FadeOutTime {type=Time}
 Debug {type=Bool}
 Gain.Drums {type=Float}
 Gain.Else {type=Float}
 Gain.Ext {type=Float}
 Gain.Stinger {type=Float}
 Identifier {type=String}
 LPF.Drums {type=Float}
 LPF.Else {type=Float}
 LPF.Ext {type=Float}
 OnBeat {type=Trigger}
 Quantization.DivisionOnStop {type=Int32}
 Quantization.DivisionStinger {type=Int32}
 Quantization.DivisionTrack {type=Int32}
 QueueTrigger {type=Trigger}
 ResetAndStop {type=Trigger}
 Wave.DrumsLoop {type=WaveAsset}
 Wave.ElseLoop {type=WaveAsset}
 Wave.ExtLoop {type=WaveAsset}
 Wave.TransitionInStinger {type=WaveAsset}
]
outputs [
 Out L {type=Audio}
 Out R {type=Audio}
]
pages [Default {
  input_defaults {
   Crossfade.FadeInTime=1
   Crossfade.FadeOutTime=1
   Debug=null
   Gain.Drums=1
   Gain.Else=1
   Gain.Ext=1
   Gain.Stinger=1
   Identifier=New Track:
   LPF.Drums=20000
   LPF.Else=20000
   LPF.Ext=20000
   OnBeat=null
   Quantization.DivisionOnStop=2
   Quantization.DivisionStinger=4
   Quantization.DivisionTrack=2
   QueueTrigger=null
   ResetAndStop=null
   Wave.DrumsLoop=null
   Wave.ElseLoop=null
   Wave.ExtLoop=null
   Wave.TransitionInStinger=null
  }
  nodes [
   n1 = /FortMetasound/Patches/IsWaveAssetValid(Trigger In=^QueueTrigger, WaveAsset=^Wave.TransitionInStinger)
   TriggerRouteTriggerRouteBool2_0 = "TriggerRoute.Trigger Route (Bool, 2):Bool"(Set 0=@n1.Is Not Valid, Set 1=@n1.Is Valid) {also {defaults {Value 0="<MetaSoundBuilderResult.SUCCEEDED: 0>"}}}
   n2 = TriggerCompare.Clamp:Bool(Compare=@TriggerRouteTriggerRouteBool2_0.On Set, A=@TriggerRouteTriggerRouteBool2_0.Value) {also {defaults {B="<MetaSoundBuilderResult.SUCCEEDED: 0>"}}}
   TriggerOnce_0 = UE.Trigger Once(Reset=^ResetAndStop, Trigger In=@n2.TRUE)
   n3 = VariableDeferredAccessor.Trigger(Variable=InitVariable.Trigger().Variable)
   TriggerAnyTriggerAny2_0 = "TriggerAny.Trigger Any (2)"(In 0=@TriggerOnce_0.Trigger Out, In 1=@n3.Value)
   n4 = &/MSP_NPCMusic_Quantizer(
    Debug=^Debug
    Identifier=^Identifier
    OnBeat=^OnBeat
    QuantizeDivision=^Quantization.DivisionTrack
    QueueTrigger=@TriggerAnyTriggerAny2_0.Out
   )
   n5 = InitVariable.Bool()
   n6 = TriggerCompare.Clamp:Bool(
    Compare=@n4.OnTriggerQuantized
    A=VariableAccessor.Bool(Variable=VariableMutator.Bool(Variable=@n5.Variable, Value=@TriggerRouteTriggerRouteBool2_0.Value).Variable).Value
   ) {also {defaults {B="<MetaSoundBuilderResult.SUCCEEDED: 0>"}}}
   ADEnvelopeADEnvelope_0 = AD Envelope.AD Envelope:Float(Decay Time=^Crossfade.FadeInTime, Trigger=@n6.TRUE) {also {defaults {Attack Time=0, Hard Reset="<MetaSoundBuilderResult.SUCCEEDED: 0>"}}}
   TriggerAnyTriggerAny2_1 = "TriggerAny.Trigger Any (2)"(In 0=@ADEnvelopeADEnvelope_0.On Done, In 1=@n6.FALSE)
   TriggerRouteTriggerRouteFloat3_0 = "TriggerRoute.Trigger Route (Float, 3):Float"(
    Set 1=@ADEnvelopeADEnvelope_0.On Trigger
    Set 0=UE.Reroute(Value=@TriggerOnce_0.Trigger Out).Value
    Value 1=UE.Subtract:Float(AdditionalOperands=@ADEnvelopeADEnvelope_0.Out Envelope) {also {defaults {PrimaryOperand=1}}}.Out
    Set 2=@TriggerAnyTriggerAny2_1.Out
   ) {also {defaults {Value 2=1}}}
   n7 = VariableAccessor.Float(Variable=VariableMutator.Float(Variable=InitVariable.Float().Variable, Value=@TriggerRouteTriggerRouteFloat3_0.Value).Variable)
   n8 = VariableAccessor.Float(Variable=@n7.Variable)
   n9 = VariableAccessor.Float(Variable=@n8.Variable)
   ADEnvelopeADEnvelope_1 = AD Envelope.AD Envelope:Float(
    Decay Time=^Crossfade.FadeOutTime
    Trigger=&MSP_NPCMusic_Quantizer(OnBeat=^OnBeat, QuantizeDivision=^Quantization.DivisionOnStop, QueueTrigger=^ResetAndStop).OnTriggerQuantized
   ) {also {defaults {Hard Reset="<MetaSoundBuilderResult.SUCCEEDED: 0>"}}}
   TriggerRouteTriggerRouteFloat2_0 = "TriggerRoute.Trigger Route (Float, 2):Float"(Set 0=@ADEnvelopeADEnvelope_1.On Trigger, Value 0=@ADEnvelopeADEnvelope_1.Out Envelope, Set 1=@TriggerAnyTriggerAny2_0.Out) {also {defaults {Value 1=1}}}
   n10 = VariableAccessor.Trigger(Variable=VariableMutator.Trigger(Value=@ADEnvelopeADEnvelope_1.On Done, Variable=InitVariable.Trigger().Variable).Variable)
   WavePlayer_0 = UE.Wave Player:Stereo(Wave Asset=^Wave.DrumsLoop, Play=@n4.OnTriggerQuantized, Stop=@n10.Value) {also {defaults {Loop="<MetaSoundBuilderResult.SUCCEEDED: 0>"}}}
   OnePoleLowPassFilter_0 = UE.One-Pole Low Pass Filter:Audio(Cutoff Frequency=^LPF.Drums, In=@WavePlayer_0.Out Left)
   OnePoleLowPassFilter_1 = UE.One-Pole Low Pass Filter:Audio(Cutoff Frequency=^LPF.Drums, In=@WavePlayer_0.Out Right)
   n11 = VariableAccessor.Trigger(Variable=@n10.Variable)
   WavePlayer_2 = UE.Wave Player:Stereo(Wave Asset=^Wave.ExtLoop, Play=@n4.OnTriggerQuantized, Stop=VariableAccessor.Trigger(Variable=@n11.Variable).Value) {also {defaults {Loop="<MetaSoundBuilderResult.SUCCEEDED: 0>"}}}
   OnePoleLowPassFilter_2 = UE.One-Pole Low Pass Filter:Audio(Cutoff Frequency=^LPF.Ext, In=@WavePlayer_2.Out Right)
   WavePlayer_1 = UE.Wave Player:Stereo(Wave Asset=^Wave.ElseLoop, Play=@n4.OnTriggerQuantized, Stop=@n11.Value) {also {defaults {Loop="<MetaSoundBuilderResult.SUCCEEDED: 0>"}}}
   OnePoleLowPassFilter_3 = UE.One-Pole Low Pass Filter:Audio(Cutoff Frequency=^LPF.Else, In=@WavePlayer_1.Out Left)
   OnePoleLowPassFilter_4 = UE.One-Pole Low Pass Filter:Audio(Cutoff Frequency=^LPF.Ext, In=@WavePlayer_2.Out Left)
   OnePoleLowPassFilter_5 = UE.One-Pole Low Pass Filter:Audio(Cutoff Frequency=^LPF.Else, In=@WavePlayer_1.Out Right)
   TriggerOnce_1 = UE.Trigger Once(Reset=^ResetAndStop, Trigger In=@n2.FALSE)
   WavePlayer_3 = UE.Wave Player:Stereo(
    Wave Asset=^Wave.TransitionInStinger
    Play=&MSP_NPCMusic_Quantizer(Debug=^Debug, OnBeat=^OnBeat, QuantizeDivision=^Quantization.DivisionStinger, QueueTrigger=@TriggerOnce_1.Trigger Out) {also {defaults {Identifier=New Track: Transition Stinger}}}.OnTriggerQuantized
   )
   AudioMixerAudioMixerStereo8_0 = "AudioMixer.Audio Mixer (Stereo, 8)"(
    Gain 0=^Gain.Stinger
    Gain 3=UE.Multiply:Float(
     AdditionalOperands=UE.Multiply:Float(AdditionalOperands=^Gain.Ext, PrimaryOperand=@n9.Value).Out
     PrimaryOperand=@TriggerRouteTriggerRouteFloat2_0.Value
    ).Out
    Gain 1=UE.Multiply:Float(
     AdditionalOperands=UE.Multiply:Float(AdditionalOperands=^Gain.Drums, PrimaryOperand=@n7.Value).Out
     PrimaryOperand=@TriggerRouteTriggerRouteFloat2_0.Value
    ).Out
    Gain 2=UE.Multiply:Float(
     AdditionalOperands=UE.Multiply:Float(AdditionalOperands=^Gain.Else, PrimaryOperand=@n8.Value).Out
     PrimaryOperand=@TriggerRouteTriggerRouteFloat2_0.Value
    ).Out
    In 1 L=@OnePoleLowPassFilter_0.Out
    In 1 R=@OnePoleLowPassFilter_1.Out
    In 3 R=@OnePoleLowPassFilter_2.Out
    In 2 L=@OnePoleLowPassFilter_3.Out
    In 3 L=@OnePoleLowPassFilter_4.Out
    In 2 R=@OnePoleLowPassFilter_5.Out
    In 0 L=@WavePlayer_3.Out Left
    In 0 R=@WavePlayer_3.Out Right
   )
   InitVariable.Float()
   TriggerOnValueChange_1 = UE.Trigger On Value Change:Float()
   PrintLogPrintLog_0 = Print Log.Print Log:Float(Trigger=@TriggerOnValueChange_1.Trigger) {also {defaults {Label="MUSIC FADE OUT: ", Value To Log=0}}}
   TriggerOnValueChange_0 = UE.Trigger On Value Change:Float()
   PrintLogPrintLog_1 = Print Log.Print Log:Float(Trigger=@TriggerOnValueChange_0.Trigger) {also {defaults {Label="MUSIC FADE IN: "}}}
   VariableMutator.Trigger(Variable=@n3.Variable, Value=@WavePlayer_3.On Play)
  ]
  links [{dst="@Out L", src=AudioMixerAudioMixerStereo8_0.Out L}, {dst="@Out R", src=AudioMixerAudioMixerStereo8_0.Out R}]
 }]
