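# MetaSound patch that plays two stereo music stems as layers, with shared Play/Stop triggers and fades.
# The stems feed a stereo mixer with per-stem volume, then an interpolated 0-to-1 gain applied to both output channels.
# Play and Stop select the fade-in or fade-out time.
# Stop reaches both wave players through a fade-length delay and trigger variable, and Play resets that delay.
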MetaSoundPatch /FortNPCMusic/Audio/Patches/MSP_LayeredTrackPlayback2
inputs [
 FadeTime.In {type=Float}
 FadeTime.Out {type=Float}
 Play {type=Trigger}
 Stem1.Stem {type=WaveAsset}
 Stem1.Volume {type=Float}
 Stem2.Stem {type=WaveAsset}
 Stem2.Volume {type=Float}
 Stop {type=Trigger}
]
outputs [
 Out L {type=Audio}
 Out R {type=Audio}
]
pages [Default {
  input_defaults {FadeTime.In=0, FadeTime.Out=0.1, Play=null, Stem1.Stem=null, Stem1.Volume=1, Stem2.Stem=null, Stem2.Volume=1, Stop=null}
  nodes [
   n1 = UE.ConversionFloatToTime(In=Value.Value:Float(Target Value=^FadeTime.In, Init Value=^FadeTime.Out, Set=^Play, Reset=^Stop).Output Value)
   n2 = Value.Value:Float(Set=^Play, Reset=^Stop) {also {defaults {Init Value=0, Target Value=1}}}
   n3 = VariableAccessor.Trigger(Variable=VariableMutator.Trigger(Variable=InitVariable.Trigger().Variable, Value=UE.Pipe(Delay Time=@n1.Out, In=@n2.On Reset, Reset=@n2.On Set).Out).Variable)
   WavePlayer_0 = UE.Wave Player:Stereo(Play=^Play, Wave Asset=^Stem1.Stem, Stop=@n3.Value) {also {defaults {Loop="<MetaSoundBuilderResult.SUCCEEDED: 0>"}}}
   WavePlayer_1 = UE.Wave Player:Stereo(Play=^Play, Wave Asset=^Stem2.Stem, Stop=VariableAccessor.Trigger(Variable=@n3.Variable).Value) {also {defaults {Loop="<MetaSoundBuilderResult.SUCCEEDED: 0>"}}}
   AudioMixerAudioMixerStereo2_0 = "AudioMixer.Audio Mixer (Stereo, 2)"(
    Gain 0=^Stem1.Volume
    Gain 1=^Stem2.Volume
    In 0 L=@WavePlayer_0.Out Left
    In 0 R=@WavePlayer_0.Out Right
    In 1 L=@WavePlayer_1.Out Left
    In 1 R=@WavePlayer_1.Out Right
   )
   n4 = UE.InterpTo:Audio(Interp Time=@n1.Out, Target=@n2.Output Value)
   Multiply_0 = UE.Multiply:Audio by Float(PrimaryOperand=@AudioMixerAudioMixerStereo2_0.Out R, AdditionalOperands=@n4.Value)
   Multiply_1 = UE.Multiply:Audio by Float(PrimaryOperand=@AudioMixerAudioMixerStereo2_0.Out L, AdditionalOperands=@n4.Value)
  ]
  links [{dst="@Out R", src=Multiply_0.Out}, {dst="@Out L", src=Multiply_1.Out}]
 }]
