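# MSP_NPCMusic_Metronome is a MetaSound patch that provides a beat clock for NPC music.
# From On Play until OnStop, a Trigger Repeat fires OnBeat at a period converted from the BPM input.
# A trigger counter tallies the beats with a reset count.
# The beat also passes through a trigger variable to a compare on Debug, which gates printing the count to the log.
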MetaSoundPatch /FortNPCMusic/Audio/Patches/MSP_NPCMusic_Metronome
inputs [
 BPM {type=Float}
 Debug {type=Bool}
 OnStop {type=Trigger}
 UE.Source.OnPlay {type=Trigger}
]
interfaces [UE.Source:1.1]
outputs [OnBeat {type=Trigger}]
pages [Default {
  input_defaults {BPM=130, Debug=null, OnStop=null, UE.Source.OnPlay=null}
  nodes [
   TriggerRepeat_0 = UE.TriggerRepeat(Stop=^OnStop, Start=^UE.Source.OnPlay, Period=UE.BPMToSeconds(BPM=^BPM).Seconds)
   n1 = VariableMutator.Trigger(Variable=InitVariable.Trigger().Variable, Value=@TriggerRepeat_0.RepeatOut)
   TriggerCounter_0 = UE.Trigger Counter(In=@TriggerRepeat_0.RepeatOut) {also {defaults {Reset Count=64}}}
   PrintLogPrintLog_0 = Print Log.Print Log:Int32(
    Trigger=TriggerCompare.Clamp:Bool(A=^Debug, Compare=VariableAccessor.Trigger(Variable=@n1.Variable).Value) {also {defaults {B="<MetaSoundBuilderResult.SUCCEEDED: 0>"}}}.True
    Value To Log=@TriggerCounter_0.Count
   ) {also {defaults {Label=Beat}}}
  ]
  links [{dst="@OnBeat", src=TriggerRepeat_0.RepeatOut}]
 }]
